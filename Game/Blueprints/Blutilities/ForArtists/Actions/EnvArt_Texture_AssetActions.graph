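# Editor texture asset utility whose CompositeMatchingNormal action assigns matching normal maps as CompositeTexture on selected textures.
# It collects selected Texture2Ds with a given suffix (default _S), plus _N textures from the selection and their folders, scanning each folder once via the Asset Registry.
# A second pass swaps the suffix for _N to find each match by name alone, fills only empty CompositeTexture slots, then saves.
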EditorUtilityBlueprint /Game/Blueprints/Blutilities/ForArtists/Actions/EnvArt_Texture_AssetActions
graphs [
 CompositeMatchingNormal {nodes [
   n1 = CompositeMatchingNormal!() -> (then, OnlyCompositeMatchingNormalForTexturesEndingInThis:String=_S) {is=title K2Node_FunctionEntry}
   > Set l_ending(l_ending:String=@n1) -> (Output_Get:String)
   > n2 = Sequence()
   > n3 = GetSelectedAssetData()
   > n4 = Set l_Assets(l_Assets:Array of Asset Data Structures=@n3) -> (Output_Get:Array of Asset Data Structures)
   > RunAssetsThroughFilter!(
    execute
    self:Asset Registry Interface=GetAssetRegistry()
    AssetDataList:Array of Asset Data Structures=@n4
    Filter:"ARFilter Structure (by ref)"=MakeARFilter!(
     PackageNames:Array of Names
     PackagePaths:Array of Names
     SoftObjectPaths:Array of Soft Object Path Structures
     ClassPaths:Array of Top Level Asset Path Structures
     RecursiveClassPathsExclusionSet:Set of Top Level Asset Path Structures
     ClassNames:Array of Names=Make Array("[0]":Name=Texture2d) -> (Array:Array of Names)
     RecursiveClassesExclusionSet:Set of Names
     bRecursivePaths:bool=false
     bRecursiveClasses:bool=false
     bIncludeOnlyOnDiskAssets:bool=false
    ) -> (ReturnValue:ARFilter Structure)
   ) -> (then)
   > n5 = For Each Loop(Array:Array of Asset Data Structures=Get l_Assets() -> (l_Assets:Array of Asset Data Structures)) -> (Array Element:Asset Data Structure)
   > n6 = Sequence()
   n7 = Break Asset Data!(AssetData=@n5) -> (AssetName, AssetClassPath)
   > n8 = Branch(Condition=EndsWith(SourceString="To String (Name)"(InName=@n7), InSuffix=_N))
   > Add!(
    execute
    TargetMap:Map of Names to Asset Data Structures=Get l_normalMapAssetData_Map() -> (l_normalMapAssetData_Map:Map of Names to Asset Data Structures)
    Key:"Name (by ref)"=@n7
    Value:"Asset Data Structure (by ref)"=@n5
   ) -> (then) {is=title K2Node_CallFunction}
   Branch(execute=@n8.else, Condition=EndsWith(SourceString="To String (Name)"(InName=@n7), InSuffix=Get l_ending() -> (l_ending:String)))
   > Add(TargetMap=Get l_specMapAssetData_Map() -> (l_specMapAssetData_Map:Map of Names to Asset Data Structures), Key=@n7, Value=@n5)
   n9 = Break Asset Data!(AssetData=@n5) -> (PackageName:Name, PackagePath:Name, AssetName, AssetClassPath)
   n10 = Branch(
    execute=@n6.then_1
    Condition=Contains Item!(TargetSet:Set of Names=Get l_pathsChecked() -> (l_pathsChecked:Set of Names), ItemToFind:"Name (by ref)"=@n9.PackagePath) -> (ReturnValue:bool) {is=title K2Node_CallFunction}
   )
   Add!(execute=@n10.else, TargetSet:Set of Names=Get l_pathsChecked(), NewItem:"Name (by ref)"=@n9.PackagePath) -> (then)
   > n11 = GetAssets!(
    execute
    self:Asset Registry Interface=GetAssetRegistry()
    Filter:"ARFilter Structure (by ref)"=MakeARFilter(
     PackagePaths=Make Array("[0]"=@n9.PackagePath)
     ClassNames=Make Array("[0]"=Texture2d)
     bRecursivePaths=false
     bRecursiveClasses=false
     bIncludeOnlyOnDiskAssets=false
    )
    bSkipARFilteredAssets:bool=true
   ) -> (then, OutAssetData:Array of Asset Data Structures, ReturnValue:bool)
   > n12 = For Each Loop(Array=@n11)
   n13 = Break Asset Data(AssetData=@n12)
   > Branch(Condition=EndsWith(SourceString="To String (Name)"(InName=@n13.AssetName), InSuffix=_N))
   > Add!(
    execute
    TargetMap:Map of Names to Asset Data Structures=Get l_normalMapAssetData_Map()
    Key:"Name (by ref)"=@n13.AssetName
    Value:"Asset Data Structure (by ref)"=@n12
   ) -> (then)
   > PrintString(InString=Append(A="Adding normal: ", B="To String (Name)"(InName=@n13.AssetName)))
   n14 = Values!(execute=@n2.then_1, TargetMap:Map of Names to Asset Data Structures=Get l_specMapAssetData_Map()) -> (then, Values:Array of Asset Data Structures)
   > n15 = For Each Loop(Array=@n14)
   n16 = LeftChop(SourceString="To String (Name)"(InName=Break Asset Data(AssetData=@n15).AssetName), Count=Len(S=Get l_ending()))
   n17 = Find(
    TargetMap:Map of Names to Asset Data Structures=Get l_normalMapAssetData_Map()
    Key:"Name (by ref)"=String To Name(InString=Append(A=@n16, B=_N))
   ) -> (Value:Asset Data Structure)
   > Branch(Condition=@n17.ReturnValue)
   > n18 = Cast To Texture2D(Object=GetAsset(InAssetData=@n15))
   > n19 = Is Valid(InputObject=Get CompositeTexture!(self:Texture Object Reference=@n18) -> (CompositeTexture:Texture Object Reference))
   n20 = Cast To Texture2D(execute=@n19.Is Not Valid, Object=GetAsset(InAssetData=@n17))
   > SetEditorProperty(Object=@n18, PropertyName=CompositeTexture, PropertyValue:Texture 2D Object Reference=@n20)
   > SaveLoadedAsset(AssetToSave=@n18, bOnlyIfIsDirty=true)
  ]}
 EventGraph {links [], nodes []}
]
variables []
