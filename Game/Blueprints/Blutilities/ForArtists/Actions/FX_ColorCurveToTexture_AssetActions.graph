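# Editor utility that bakes selected CurveLinearColor assets into small one-row lookup textures for artists.
# ColorCurveToTexture filters the selected assets by class and passes them to an embedded Python script.
# The script samples each curve into an RGBA8 canvas render target, skipping both end pixels, then converts it to a Texture2D.
# Textures save beside their curves with C_ renamed to T_, clamped addressing, and overwrite of existing assets.
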EditorUtilityBlueprint /Game/Blueprints/Blutilities/ForArtists/Actions/FX_ColorCurveToTexture_AssetActions
graphs [
 ColorCurveToTexture {nodes [
   ColorCurveToTexture() {is=title K2Node_FunctionEntry}
   > n1 = GetSelectedAssets()
   > n2 = Filter by Class!(
    execute
    TargetArray:Array of Object References=@n1
    ObjectClass:Object Class Reference=/Script/Engine.CurveLinearColor
    FilterType:EEditorScriptingFilterType Enum
   ) -> (then, ReturnValue:Array of Curve Linear Color Object References)
   > n3 = Set Array(Array:Array of Curve Linear Color Object References=@n2) -> (Output_Get:Array of Curve Linear Color Object References)
   > Execute Python Script!(
    execute
    PythonScript="import os\r\nimport unreal as ue\r\nwco = ue.EditorLevelLibrary.get_editor_world()\r\nueRL = ue.RenderingLibrary\r\nar = ue.AssetRegistryHelpers.get_asset_registry()\r\n\r\n##eul = ue.EditorUtilityLibrary() ##use editor selection\r\n##sel = eul.get_selected_assets()\r\n\r\nsel = curvelinearcolor_array ##use BP input\r\nat = ue.AssetToolsHelpers.get_asset_tools()\r\nsize_x = 64\r\n\r\n## Create a CanvasRenderTarget2D object\r\ndef make_CanvasRT2D(width=size_x):\r\n    crt = ue.CanvasRenderTarget2D()\r\n    crt.set_editor_property(\"size_x\", size_x)\r\n    crt.set_editor_property(\"size_y\", 1)\r\n    crt.set_editor_property(\"srgb\", 0)\r\n    crt.set_editor_property(\"address_x\", ue.TextureAddress.TA_CLAMP)  \r\n    crt.set_editor_property(\"address_y\", ue.TextureAddress.TA_CLAMP)  \r\n    crt.set_editor_property('render_target_format', ue.TextureRenderTargetFormat.RTF_RGBA8)\r\n    return crt\r\n\r\n## Define the update delegate (then draw box-pixels in to the canvas)\r\ndef update_canvas(canvas, width, height): \r\n    for i in range(1,width-1):\r\n        ## curve is class CurveLinearColor\r\n        col = curve.get_linear_color_value(float(i)*(1.0/float(width)))\r\n        canvas.draw_box(\r\n            screen_position=ue.Vector2D(i, 0),\r\n            screen_size=ue.Vector2D(1, 1),\r\n            thickness=1.0,\r\n            render_color=ue.LinearColor(col.r, col.g, col.b, col.a) ## sampled color\r\n        )\r\n\r\n## loop over selection, do the deed\r\nif len(sel)>0:\r\n    for curve in sel: ## prefiltered by BP\r\n        ## New Canvas, bind function to the update delegate, then update\r\n        crt = make_CanvasRT2D(width=size_x)\r\n        crt.on_canvas_render_target_update.add_callable(update_canvas)\r\n        crt.update_resource()\r\n        ## path and input/output name\r\n        assetpath, curvename = os.path.split(curve.get_path_name())\r\n        texname = curvename.split(\".\")[0].replace('C_',\"T_\")\r\n        outassetpathname = assetpath+\"/\"+texname+\".\"+texname\r\n        mytex = ue.Texture2DFactoryNew()\r\n        print(f\"writing out {texname}\")\r\n        if(ue.EditorAssetLibrary.does_asset_exist(outassetpathname)):\r\n            print(\"exists\") ## get existing Texture2D and overwrite\r\n            asset_data = ar.get_asset_by_object_path(outassetpathname)\r\n            if asset_data:\r\n                mytex_asset = asset_data.get_asset()\r\n                if isinstance(mytex_asset, ue.Texture2D):\r\n                    print(f\"Successfully loaded Texture2D: {mytex.get_name()}\")\r\n                    print(f\"Texture size: {mytex_asset.blueprint_get_size_x()}x{mytex_asset.blueprint_get_size_y()}\")\r\n                else:\r\n                    print(f\"Asset at {outassetpathname} is not a Texture2D.\")\r\n            else:\r\n                print(f\"No asset found at {outassetpathname}\")\r\n            opts = ue.GeometryScriptCreateNewTexture2DAssetOptions(True) ## overwrite in place\r\n            ueRL.convert_render_target_to_texture2d_editor_only(wco, crt, mytex_asset)\r\n            ue.GeometryScript_NewAssetUtils.create_new_texture2d_asset(mytex_asset, outassetpathname, opts, debug=None)\r\n        else:\r\n            mytex_asset = at.create_asset(texname, assetpath, ue.Texture2D, mytex)\r\n            ueRL.convert_render_target_to_texture2d_editor_only(wco, crt, mytex_asset)\r\n        mytex_asset.set_editor_property(\"address_x\", ue.TextureAddress.TA_CLAMP)  \r\n        mytex_asset.set_editor_property(\"address_y\", ue.TextureAddress.TA_CLAMP) \r\n        ue.EditorAssetLibrary.save_asset(outassetpathname)\r\n\r\n\r\n"
    CurvelinearcolorArray:Array of Curve Linear Color Object References=@n3
   ) -> (then, ReturnValue) {inputs [CurvelinearcolorArray]}
   n4 = Get Class(Object="Get (a copy)"(Array:Array of Curve Linear Color Object References=@n3) -> (Output:Curve Linear Color Object Reference)) -> (ReturnValue:Curve Linear Color Class Reference)
   n5 = Length(TargetArray:Array of Curve Linear Color Object References=@n3)
   PrintString(InString=Append!(A="Class = ", B=Get Class Display Name(Class=@n4), C:String="  Num = ", D:String="To String (Integer)"(InInt=@n5)) -> (ReturnValue))
  ]}
 EventGraph {links [], nodes []}
]
variables [Array {type="array<CurveLinearColor>", default []}]
